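# Looping ambient Sound Cue for a bubbling pot spot emitter, routed to the ambient sound class with a reverb send.
# A mixer layers a lava bubbles loop with one of five randomly chosen viscous bubbling loops, the latter pitch and volume modulated.
# Three nested, endlessly looping envelopers add a pitch curve, irregular volume wobble and a shorter periodic swell to vary the loop.
# Concurrency is capped at two voices.
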SoundCue /Game/Sounds/Helios/Environment_Actors/BubblingPot/Spot_Bubbling_Pot_Loop_Cue
root=Enveloper_4
nodes [
 {
  id=Enveloper_0
  children [Mixer_0]
  properties {
   bLoop=true
   bLoopIndefinitely=true
   loopEnd=60
   pitchCurve {editorCurveData {keys [
      {interpMode=RCIM_Cubic, time=0}
      <at 15 0.85 RCIM_Cubic>
      {arriveTangent=0.0091, interpMode=RCIM_Cubic, leaveTangent=0.0091, tangentMode=RCTM_User, time=30}
      {interpMode=RCIM_Cubic, time=45}
      {time=60}
     ]}}
   volumeCurve {editorCurveData {keys [<at 0 1>]}}
  }
 }
 {
  id=Enveloper_1
  children [Enveloper_0]
  properties {
   bLoop=true
   bLoopIndefinitely=true
   loopEnd=60
   pitchCurve {editorCurveData {keys [{time=0}]}}
   volumeCurve {editorCurveData {keys [
      {arriveTangent=-0.01, interpMode=RCIM_Cubic, time=0, value=1}
      {arriveTangent=-0.0035, interpMode=RCIM_Cubic, leaveTangent=-0.0035, tangentMode=RCTM_User, time=4.5428, value=0.8801}
      {arriveTangent=0.0016, interpMode=RCIM_Cubic, leaveTangent=0.0016, tangentMode=RCTM_User, time=7.4135, value=0.9776}
      {arriveTangent=0.0008, interpMode=RCIM_Cubic, leaveTangent=0.0008, tangentMode=RCTM_User, time=11.1077, value=0.8936}
      {arriveTangent=0.0016, interpMode=RCIM_Cubic, leaveTangent=0.0016, tangentMode=RCTM_User, time=12.9775, value=0.8792}
      {arriveTangent=0.0009, interpMode=RCIM_Cubic, leaveTangent=0.0009, time=14.965, value=0.9463}
      {arriveTangent=-0.0011, interpMode=RCIM_Cubic, leaveTangent=-0.0011, tangentMode=RCTM_User, time=18.5868, value=0.8842}
      {arriveTangent=0.0002, interpMode=RCIM_Cubic, leaveTangent=0.0002, time=23.9559, value=0.9962}
      {arriveTangent=0.0016, interpMode=RCIM_Cubic, leaveTangent=0.0016, tangentMode=RCTM_User, time=28.034, value=0.8864}
      {arriveTangent=-0.0003, interpMode=RCIM_Cubic, leaveTangent=-0.0003, time=31.2025, value=0.961}
      {arriveTangent=0.0009, interpMode=RCIM_Cubic, leaveTangent=0.0009, tangentMode=RCTM_User, time=34.9227, value=0.8844}
      {arriveTangent=-0.0008, interpMode=RCIM_Cubic, leaveTangent=-0.0008, time=37.93, value=0.9962}
      <at 42.0081 0.879>
      {arriveTangent=0.0024, interpMode=RCIM_Cubic, leaveTangent=0.0024, tangentMode=RCTM_User, time=44.5668, value=0.8842}
      {arriveTangent=-0.0006, interpMode=RCIM_Cubic, leaveTangent=-0.0006, time=47.958, value=0.9776}
      {arriveTangent=-0.0056, interpMode=RCIM_Cubic, leaveTangent=-0.0056, tangentMode=RCTM_User, time=53.1343, value=0.8792}
      <at 55.3163 0.9711>
      {arriveTangent=-0.0003, interpMode=RCIM_Cubic, leaveTangent=-0.0003, tangentMode=RCTM_User, time=57.7536, value=0.8789}
      <at 60 1 RCIM_Cubic>
     ]}}
  }
 }
 {
  id=Enveloper_4
  children [Enveloper_1]
  properties {
   bLoop=true
   bLoopIndefinitely=true
   loopEnd=20
   pitchCurve {editorCurveData {keys [{time=0}]}}
   volumeCurve {editorCurveData {keys [
      {arriveTangent=-0.01, interpMode=RCIM_Cubic, time=0, value=1}
      <at 10 1.2 RCIM_Cubic>
      {interpMode=RCIM_Cubic, leaveTangent=0.01, time=20, value=1}
     ]}}
  }
 }
 {id=Mixer_0, children [Modulator_0,WavePlayer_3], properties {inputVolume [1,1]}}
 {id=Modulator_0, children [Random_0], properties {pitchMax=1.1}}
 {id=Random_0, children [WavePlayer_0,WavePlayer_1,WavePlayer_2,WavePlayer_4,WavePlayer_5], properties {weights [1,1,1,1,1]}}
 {
  id=WavePlayer_0
  properties {bLooping=true, soundWaveAssetPtr=/Game/Sounds/Helios/Biomes/Underworld/Spots/Spot_Liquid_Viscous_Flow_Bubbling_Slowly_B_05}
 }
 {
  id=WavePlayer_1
  properties {bLooping=true, soundWaveAssetPtr=/Game/Sounds/Helios/Biomes/Underworld/Spots/Spot_Liquid_Viscous_Flow_Bubbling_Slowly_B_04}
 }
 {
  id=WavePlayer_2
  properties {bLooping=true, soundWaveAssetPtr=/Game/Sounds/Helios/Biomes/Underworld/Spots/Spot_Liquid_Viscous_Flow_Bubbling_Slowly_B_03}
 }
 {
  id=WavePlayer_3
  properties {bLooping=true, soundWaveAssetPtr=/Game/Sounds/Helios/Environment_Actors/Mephisto_Hades/Spot_Mephisto_Lava_Bubbles_Loop}
 }
 {
  id=WavePlayer_4
  properties {bLooping=true, soundWaveAssetPtr=/Game/Sounds/Helios/Biomes/Underworld/Spots/Spot_Liquid_Viscous_Flow_Bubbling_Slowly_B_02}
 }
 {
  id=WavePlayer_5
  properties {bLooping=true, soundWaveAssetPtr=/Game/Sounds/Helios/Biomes/Underworld/Spots/Spot_Liquid_Viscous_Flow_Bubbling_Slowly_B_01}
 }
]
properties {
 attenuationSettings=/Game/Sounds/Helios/Biomes/Underworld/Spots/Fort_Amb_Water_Liquid_Viscous_Flowing_Att
 bOverrideConcurrency=true
 maxDistance=1250
 soundClassObject=&Fort_Ambient_Athena
 volumeMultiplier=0.301
 attenuationOverrides {falloffDistance=2000, nonSpatializedRadiusStart=700, stereoSpread=1000, attenuationShapeExtents {x=925}}
 concurrencyOverrides {maxCount=2, voiceStealReleaseTime=2}
 soundSubmixSends [{maxSendLevel=1.2, minSendLevel=1, sendLevel=1.2, soundSubmix=&Fortnite_Reverb_Default}]
}
